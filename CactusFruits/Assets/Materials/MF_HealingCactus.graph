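# Material function for the healing cactus that animates cactus and fruit bursting and regrowth, plus fruit pulsing and energy glow.
# Custom primitive data (start time, fruit state, vehicle flag) drives timelines sampling curve-atlas rows for world position offset, opacity mask, color and emissive.
# UV channels supply part masks, Houdini-baked pivots and random pop staggering.
# It also adds subsurface shading and spherical-harmonics Slurp coloring; a static switch bypasses the animation.
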MaterialFunction /CactusFruits/Assets/Materials/MF_HealingCactus
graph {nodes [
  n1 = Subtract(A=NamedRerouteUsage(), B=ScalarParameter() {group=HealingCactus, parameterName=HC_Cactus_BurstDelay}) {constB=0.5}
  n2 = ScalarParameter() {defaultValue=0.6, group=HealingCactus, parameterName=HC_Cactus_BurstDuration}
  n3 = Saturate(None=StaticSwitch(True=@n1, False=Divide(A=@n1, B=@n2), Value=NamedRerouteUsage()))
  n4 = CurveAtlasRowParameter(CurveTime=@n3) {
   atlas=/CactusFruits/Assets/Textures/CA_Wasteland_HealingCactus
   curve=/CactusFruits/Assets/Textures/C_Wasteland_HealingCactus
   group=HealingCactus
   parameterName=HC_Cactus_AnimCurve
  }
  ConstantBiasScale(None=@n4.G) {bias=-0.5, scale=2}
  Constant()
  Frac(None=Multiply(A=Time()) {constB=0.5})
  Frac(None=Time())
  n5 = SetMaterialAttributes!(
   MaterialAttributes=FunctionInput() {inputType=FunctionInput_MaterialAttributes}
   ShadingModel=ShadingModel() {shadingModel=MSM_Subsurface}
   SubsurfaceColor=NamedRerouteUsage()
   Opacity=NamedRerouteUsage()
  ) {attributeSetTypes [#1, #2, #3]}
  n6 = GetMaterialAttributes(None=@n5) {attributeGetTypes [
    #4
    #5
    #6
    #7
    #2
    #8
   ]}
  n7 = SetMaterialAttributes!(
   MaterialAttributes=@n6.MaterialAttributes
   WorldPositionOffset=NamedRerouteUsage()
   BaseColor=NamedRerouteUsage()
   OpacityMask=NamedRerouteUsage()
   EmissiveColor=NamedRerouteUsage()
   Normal
   Roughness=NamedRerouteUsage()
   Specular=NamedRerouteUsage()
   ShadingModel
  ) {attributeSetTypes [
    #5
    #4
    #9
    #6
    #10
    #7
    #8
    #1
   ]}
  FunctionOutput(None=StaticSwitchParameter(True=@n7, False=@n5) {group=HealingCactus, parameterName=UseHealingCactusAnim})
  MaterialFunctionCall() {materialFunction=/Engine/Functions/Engine_MaterialFunctions02/PivotPainter/PivotPainter_PerObjectData}
  n8 = VectorParameter() {group=Slurp, parameterName=SHMax, defaultValue=<rgba 1.25 0.13 0.18 0.13>}
  n9 = VectorParameter() {group=Slurp, parameterName=SHMin, defaultValue=<rgba 1 -0.16 -0.17 -0.15>}
  n10 = AppendVector(A=@n9.RGB, B=@n9.RGB)
  n11 = Subtract(A=AppendVector(A=@n8.RGB, B=@n8.RGB), B=@n10)
  n12 = VertexColor()
  n13 = AppendVector(A=@n12, B=@n12)
  n14 = Multiply(A=MaterialFunctionCall(X=@n12.G, Y=@n12.G, Z=@n12.G) {materialFunction=&MakeFloat3}.Result)
  MaterialExpressionStaticSwitchParameter_8 = StaticSwitchParameter(True=Add(A=Multiply(A=@n11, B=@n13), B=@n10), False=AppendVector(A=Add(A=@n12.R), B=Add(A=@n14) {constB=-0.5})) {group=Slurp, parameterName=UseCustomSH Fit}
  n15 = Multiply(
   A=MaterialFunctionCall() {materialFunction=&CameraDirectionVector}.Result
   B=ScalarParameter() {defaultValue=1, group=Slurp, parameterName=Slurp_DirectionVectorFlip}
  )
  n16 = StaticSwitchParameter(
   True=Normalize(VectorInput=VectorParameter() {group=Slurp, parameterName=Slurp_CustomDirectionVector, defaultValue=<rgba 0 0 -1 1>}.RGB)
   False=Transform(None=ComponentMask(None=@n15) {b=true, g=true, r=true}) {transformSourceType=TRANSFORMSOURCE_World, transformType=TRANSFORM_Local}
  ) {group=Slurp, parameterName=UseCustomDirection}
  n17 = VertexNormalWS()
  MaterialExpressionStaticSwitchParameter_1 = StaticSwitchParameter(
   True=MaterialFunctionCall!(
    VectorToRotate=@n16
    PreSkinnedLocalNormal=NamedRerouteUsage()
    VertexlNormalWS=Transform(None=@n17) {transformSourceType=TRANSFORMSOURCE_World, transformType=TRANSFORM_Local}
   ) {materialFunction=/Game/Packages/Fortress_SharedMaterials/Base_Material_Functions/RotateVector_FromSkinning}.RotatedVectorFromSkinning
   False=@n16
  ) {defaultValue=true, group=Slurp, parameterName=UseRotateVector_FromSkinning_SH_Dir}
  n18 = Custom!(SH=@MaterialExpressionStaticSwitchParameter_8, Direction=@MaterialExpressionStaticSwitchParameter_1) {
   code="//SH decode,  Direction = LightDir, SH = 2band coefficients. (encoded in Y-up)\r\nDirection = normalize(Direction);\r\nfloat4 shEvalDirection;\r\nshEvalDirection.x = 0.282095;\r\nshEvalDirection.y =0.488603 * Direction.z;\r\nshEvalDirection.z = 0.488603 * Direction.y;\r\nshEvalDirection.w =-0.488603 * Direction.x;\r\nreturn saturate(dot(SH, shEvalDirection));"
   description=SphericalHarmonicsEval
   outputType=CMOT_Float1
   inputs [
    {inputName=SH, input {expression=&self:MaterialExpressionStaticSwitchParameter_8}}
    {inputName=Direction, input {expression=&self:MaterialExpressionStaticSwitchParameter_1}}
   ]
  }
  n19 = OneMinus(None=MaterialFunctionCall(
    Input=@n18
    Input Low=ScalarParameter() {group=Slurp, parameterName=SHContrast_LowValue}
    Input High=ScalarParameter() {defaultValue=1, group=Slurp, parameterName=SHContrast_HighValue}
    Target Low=Constant()
    Target High=Constant() {r=1}
   ) {materialFunction=&RemapValueRange}.Result)
  n20 = VertexInterpolator(VS=Power(Base=Saturate(None=@n19), Exp=ScalarParameter() {defaultValue=1, group=Slurp, parameterName=SHPow}))
  n21 = QualitySwitch(Default=@n20.PS, Low=@n20.PS)
  n22 = LinearInterpolate(
   A=LinearInterpolate(
    A=VectorParameter() {group=Slurp, parameterName=Slurp_InnerColor_Immature, defaultValue=<rgba 0.2296 0.4271 0.1974 1>}.RGB
    B=VectorParameter() {group=Slurp, parameterName=Slurp_InnerColor, defaultValue=<rgba 0.0274 0.2254 0.5365 1>}.RGB
    Alpha=NamedRerouteUsage()
   )
   B=NamedRerouteUsage()
   Alpha=NamedRerouteDeclaration(None=ShadingPathSwitch(Default=@n21, Deferred=@n21, Forward=@n20.PS, Mobile=@n20.PS)) {name=SH, nodeColor=<rgba 0.5938 0 1 1>}
  )
  Multiply(
   A=LinearInterpolate(
    A=@n22
    B=VectorParameter() {group=Slurp, parameterName=Slurp_InnerScatteringColor, defaultValue=<rgba 0.1232 0.6518 1 1>}.RGB
    Alpha=NamedRerouteUsage()
   )
   B=LinearInterpolate(
    A=Constant3Vector() {constant=<rgba 1 1 1 0>}
    B=VectorParameter() {group=Slurp, parameterName=Slurp_InnerBackScatteringMultiplier, defaultValue=<rgba 0 0 0 1>}.RGB
    Alpha=NamedRerouteUsage()
   )
  )
  NamedRerouteDeclaration(None=ComponentMask(None=TextureCoordinate() {coordinateIndex=3}) {g=true}) {name=RandomMask, nodeColor=<rgba 0.0781 1 0 1>}
  NamedRerouteDeclaration(None=@n6.WorldPositionOffset) {name=WPO_IN, nodeColor=<rgba 0.0781 0 1 1>}
  n23 = NamedRerouteUsage()
  n24 = OneMinus(None=ScalarParameter() {defaultValue=1, group=HealingCactus, parameterName=HC_Fruits_EmissiveBaseColorRatio})
  NamedRerouteDeclaration(None=LinearInterpolate(A=NamedRerouteUsage(), B=Multiply(A=@n23, B=@n24), Alpha=NamedRerouteUsage()) {constB=0}) {name=BaseColor_OUT, nodeColor=<rgba 0 1 0.8828 1>}
  NamedRerouteDeclaration(None=@n6.EmissiveColor) {name=Emissive_IN, nodeColor=<rgba 1 0 0.1172 1>}
  n25 = NamedRerouteUsage()
  n26 = NamedRerouteUsage()
  n27 = ScalarParameter() {defaultValue=0.8, group=HealingCactus, parameterName=HC_Fruits_EmissiveBaseColorRatio}
  n28 = Add(A=NamedRerouteUsage(), B=Multiply(A=Multiply(A=@n26, B=@n27), B=NamedRerouteUsage()))
  NamedRerouteDeclaration(None=Add(A=NamedRerouteUsage(), B=Add(A=@n25, B=@n28))) {name=Emissive_OUT, nodeColor=<rgba 1 0.6563 0 1>}
  NamedRerouteDeclaration(None=LocalPosition().XYZ) {name=PS-LocalPos, nodeColor=<rgba 0.3828 0 1 1>}
  n29 = NamedRerouteUsage()
  n30 = Multiply(
   A=ScalarParameter() {defaultValue=1, group=HealingCactus, parameterName=HC_Fruits_Regrowth_Brightness}
   B=MaterialFunctionCall(Hue Shift Percentage=ScalarParameter() {defaultValue=-0.35, group=HealingCactus, parameterName=HC_Fruits_Regrowth_HueShift}, Texture=@n29) {materialFunction=&HueShift}.Result
  )
  n31 = Subtract(A=NamedRerouteUsage(), B=ScalarParameter() {defaultValue=1.2, group=HealingCactus, parameterName=HC_Fruits_RegrowthDelay})
  n32 = Divide(A=@n31, B=ScalarParameter() {defaultValue=1.6, group=HealingCactus, parameterName=HC_Fruits_RegrowthColorShiftDuration})
  n33 = NamedRerouteUsage()
  n34 = Saturate(None=Subtract(A=StaticSwitch(True=@n31, False=@n32, Value=@n33), B=NamedRerouteUsage()) {desc=Pops in random order})
  n35 = Multiply(A=NamedRerouteUsage(), B=Saturate(None=NamedRerouteUsage()))
  NamedRerouteDeclaration(None=LinearInterpolate(A=NamedRerouteUsage(), B=LinearInterpolate(A=@n30, B=@n29, Alpha=@n34), Alpha=@n35)) {name=BaseColor_Growth, nodeColor=<rgba 1 0 0.7969 1>}
  NamedRerouteDeclaration(None=@n6.SubsurfaceColor) {name=SSSColor_IN, nodeColor=<rgba 0.2422 0 1 1>}
  NamedRerouteDeclaration() {name=Debug, nodeColor=<rgba 1 0 0.75 1>}
  n36 = ConstantBiasScale(None=@n4.R) {bias=-0.5, scale=2}
  n37 = Constant()
  n38 = Saturate(None=SphereMask(A=NamedRerouteUsage(), B=Constant3Vector() {constant=<rgba 0 0 70 0>}, Radius=LinearInterpolate(Alpha=@n3) {constB=200}) {attenuationRadius=120, hardnessPercent=0})
  n39 = NamedRerouteUsage()
  n40 = Add(A=NamedRerouteUsage(), B=VectorParameter() {group=HealingCactus, parameterName=HC_Cactus_BurstOffset, defaultValue=<rgba -15 0 0 1>}.RGB)
  n41 = Multiply(
   A=LinearInterpolate(
    B=MaterialFunctionCall(X=@n36, Y=@n36, Z=@n37) {materialFunction=&MakeFloat3}.Result
    Alpha=Multiply(A=@n38, B=@n39)
   ) {constAlpha=1}
   B=Multiply(A=Normalize(VectorInput=@n40), B=ScalarParameter() {defaultValue=55, group=HealingCactus, parameterName=HC_Cactus_BurstIntensity}) {constB=10}
  )
  n42 = NamedRerouteUsage()
  NamedRerouteDeclaration(None=Transform(None=Multiply(A=@n41, B=@n42)) {desc=Animate in object space then transform to world, transformSourceType=TRANSFORMSOURCE_Local}) {name=WPO_Burst_Cactus, nodeColor=<rgba 0.0547 1 0 1>}
  NamedRerouteDeclaration(None=StaticBoolParameter() {group=HealingCactus_DEBUG, parameterName=UseDEBUG}) {name=UseDebug, nodeColor=<rgba 0.5 1 0 1>}
  n43 = NamedRerouteUsage()
  n44 = ScalarParameter() {defaultValue=0.65, group=HealingCactus, parameterName=HC_SSSBrightness}
  n45 = ScalarParameter() {defaultValue=0.3, group=HealingCactus, parameterName=HC_SSSBrightness_Fruits}
  NamedRerouteDeclaration(None=LinearInterpolate(A=Multiply(A=@n43, B=@n44), B=Multiply(A=@n43, B=@n45), Alpha=NamedRerouteUsage()) {constA=1}) {name=SSSColor_OUT, nodeColor=<rgba 0.5703 0 1 1>}
  n46 = Subtract(A=Add(A=WorldPosition().XYZ, B=NamedRerouteUsage()), B=NamedRerouteUsage())
  n47 = Multiply(A=@n46) {constB=-1}
  n48 = Multiply(A=@n46) {constB=0}
  n49 = Saturate(None=NamedRerouteUsage())
  n50 = Subtract(A=NamedRerouteUsage(), B=ScalarParameter() {defaultValue=1.2, group=HealingCactus, parameterName=HC_Fruits_RegrowthDelay})
  n51 = ScalarParameter() {defaultValue=2.75, group=HealingCactus, parameterName=HC_Fruits_RegrowthDuration}
  n52 = Saturate(None=StaticSwitch(True=@n50, False=Divide(A=@n50, B=@n51), Value=NamedRerouteUsage()))
  n53 = ScalarParameter() {defaultValue=2, group=HealingCactus, parameterName=HC_Fruits_RegrowthInterval}
  n54 = NamedRerouteUsage()
  n55 = Multiply(A=@n53, B=Frac(None=Add(A=@n54) {constB=0.3847}))
  n56 = Subtract(A=Multiply(A=Multiply(A=@n49, B=@n52), B=Multiply(A=@n53) {constB=2}) {constB=2}, B=@n55)
  n57 = CurveAtlasRowParameter(CurveTime=NamedRerouteDeclaration(None=Saturate(None=@n56)) {name=FruitRegrowProgress, nodeColor=<rgba 0 1 0.2266 1>}) {
   atlas=&CA_Wasteland_HealingCactus
   curve=/CactusFruits/Assets/Textures/C_Wasteland_HealingCactus_Fruits
   defaultValue=1
   group=HealingCactus
   parameterName=HC_Fruits_AnimCurve
  }
  NamedRerouteDeclaration(None=Multiply(A=LinearInterpolate(A=@n47, B=@n48, Alpha=@n57.G) {constA=-1, constB=0}, B=NamedRerouteUsage())) {name=WPO_Regrowth_Fruit, nodeColor=<rgba 1 0.0469 0 1>}
  n58 = NamedRerouteUsage()
  n59 = ScalarParameter() {defaultValue=1, group=HealingCactus, parameterName=HC_Cactus_RegrowthDuration}
  n60 = Multiply(A=Divide(A=@n58, B=@n59)) {constB=1.25}
  n61 = Divide(A=NamedRerouteUsage()) {constB=1024}
  n62 = Multiply(A=@n61, B=ScalarParameter() {defaultValue=0.1, group=HealingCactus, parameterName=HC_Cactus_RegrowthWobbleIntensity}) {constB=0.2, desc=TO-DO}
  n63 = CurveAtlasRowParameter(CurveTime=Saturate(None=Subtract(A=@n60, B=@n62) {constB=0})) {
   atlas=&CA_Wasteland_HealingCactus
   curve=&C_Wasteland_HealingCactus
   group=HealingCactus
   parameterName=HC_Cactus_AnimCurve
  }
  n64 = ConstantBiasScale(None=@n63.B) {bias=-0.5, scale=2}
  n65 = Subtract(A=WorldPosition().XYZ, B=NamedRerouteUsage())
  n66 = Multiply(
   A=MaterialFunctionCall(X=@n64, Y=@n64, Z=@n64) {materialFunction=&MakeFloat3}.Result
   B=@n65
  )
  n67 = NamedRerouteUsage()
  n68 = Subtract(A=NamedRerouteUsage(), B=@n67)
  n69 = Subtract(A=@n58, B=ScalarParameter() {defaultValue=0.2, group=HealingCactus, parameterName=HC_Cactus_ArmsRegrowthDelay}) {constB=0.5}
  n70 = Subtract(A=Multiply(A=Divide(A=@n69, B=@n59)), B=Multiply(A=NamedRerouteUsage()) {constB=0.2})
  n71 = Multiply(A=NamedRerouteUsage(), B=@n61)
  n72 = CurveAtlasRowParameter(CurveTime=Saturate(None=Subtract(A=@n70, B=@n71) {constB=0})) {
   atlas=&CA_Wasteland_HealingCactus
   curve=&C_Wasteland_HealingCactus
   group=HealingCactus
   parameterName=HC_Cactus_AnimCurve
  }
  n73 = ConstantBiasScale(None=@n72.A) {bias=-0.5, scale=2}
  n74 = Multiply(
   A=@n65
   B=MaterialFunctionCall(X=@n73, Y=@n73, Z=@n73) {materialFunction=&MakeFloat3}.Result
  )
  NamedRerouteDeclaration(None=Add(A=Multiply(A=@n66, B=@n67), B=Multiply(A=@n68, B=@n74))) {name=WPO_Regrowth_Cactus, nodeColor=<rgba 0.0547 1 0 1>}
  NamedRerouteDeclaration(None=@n55) {name=RandomMask_Pop, nodeColor=<rgba 1 0 0.2344 1>}
  n75 = ScalarParameter() {defaultValue=0.5, group=HealingCactus, parameterName=HC_BaseColorbrightness}
  NamedRerouteDeclaration(None=LinearInterpolate(A=@n6.BaseColor, B=Multiply(A=@n6.BaseColor, B=@n75) {constB=0.8}, Alpha=NamedRerouteUsage())) {name=BaseColor_IN, nodeColor=<rgba 0 1 0.1797 1>}
  n76 = MaterialFunctionCall() {materialFunction=&BoundingBoxBased_0-1_UVW}
  NamedRerouteDeclaration(None=@n76.RGB) {name=BBox_UVW, nodeColor=<rgba 0 0.1094 1 1>}
  n77 = NamedRerouteUsage()
  n78 = ScalarParameter() {defaultValue=1, group=HealingCactus, parameterName=HC_Cactus_EnergyDuration}
  n79 = Saturate(None=StaticSwitch(True=@n77, False=Divide(A=@n77, B=@n78), Value=NamedRerouteUsage()))
  n80 = Subtract(
   A=Multiply(A=NamedRerouteDeclaration(None=@n79) {name=EnergyTimeline, nodeColor=<rgba 0 1 0.2734 1>}) {constB=2.5}
   B=NamedRerouteDeclaration(None=@n76.B) {name=BBox_Z, nodeColor=<rgba 1 0.6328 0 1>}
  )
  n81 = Constant() {r=0.5}
  n82 = OneMinus(None=Divide(A=Distance(A=@n80, B=@n81)) {constB=0.6})
  n83 = Saturate(None=Power(Base=Saturate(None=@n82)) {constExponent=3})
  n84 = NamedRerouteUsage()
  n85 = Multiply(
   A=LinearInterpolate(A=LinearInterpolate(A=@n83, B=@n84) {constAlpha=0.25}, B=Multiply(A=@n83, B=@n84), Alpha=NamedRerouteUsage())
   B=NamedRerouteUsage()
  )
  n86 = VectorParameter() {group=HealingCactus, parameterName=HC_Cactus_EnergyEmissiveColor, defaultValue=<rgba 0 1.6549 3 1>}
  n87 = Multiply(A=NamedRerouteUsage()) {constB=-1}
  NamedRerouteDeclaration(None=Multiply(A=Multiply(A=@n85, B=@n86.RGB), B=Saturate(None=@n87))) {name=Emissive_Cactus, nodeColor=<rgba 0.7344 1 0 1>}
  n88 = TextureCoordinate() {coordinateIndex=1}
  n89 = Step(Y=ComponentMask(None=@n88) {r=true}) {constX=0.5, constY=0.5}
  n90 = Multiply(A=@n89, B=Step(X=ComponentMask(None=@n88) {g=true}) {constY=0.5})
  NamedRerouteDeclaration(None=OneMinus(None=NamedRerouteDeclaration(None=@n90) {name=FruitMask, nodeColor=<rgba 0 1 0.6016 1>})) {name=CactusMask, nodeColor=<rgba 1 0 0.0469 1>}
  NamedRerouteDeclaration(None=OneMinus(None=@n89)) {name=MainBranchMask, nodeColor=<rgba 1 0.1172 0 1>}
  NamedRerouteDeclaration(None=@n6.Roughness) {name=Roughness_IN, nodeColor=<rgba 1 0 0.7266 1>}
  n91 = NamedRerouteUsage()
  n92 = VectorParameter() {group=HealingCactus, parameterName=HC_Fruits_PulseEmissiveColor, defaultValue=<rgba 0 1 0.8623 1>}
  NamedRerouteDeclaration(None=Multiply(A=Multiply(A=@n91, B=@n92.RGB), B=ScalarParameter() {defaultValue=0.5, group=HealingCactus, parameterName=HC_Fruits_PulseGlowIntensity})) {name=EmissivePulse_Fruits, nodeColor=<rgba 0.1484 1 0 1>}
  n93 = Fresnel() {baseReflectFraction=0, exponent=1}
  n94 = Power(
   Base=OneMinus(None=NamedRerouteDeclaration(None=@n93) {name=Fresnel, nodeColor=<rgba 0 0.625 1 1>})
   Exp=ScalarParameter() {defaultValue=8, group=HealingCactus, parameterName=HC_Fruits_PulseGlowPower}
  ) {constExponent=3}
  n95 = Subtract(A=NamedRerouteUsage(), B=ScalarParameter() {defaultValue=4, group=HealingCactus, parameterName=HC_Fruits_RegrowthPulseDelay})
  n96 = Divide(A=@n95, B=ScalarParameter() {defaultValue=3, group=HealingCactus, parameterName=HC_Fruits_RegrowthPulseDuration})
  n97 = NamedRerouteUsage()
  n98 = LinearInterpolate(B=Saturate(None=StaticSwitch(True=@n95, False=@n96, Value=@n97)), Alpha=Saturate(None=NamedRerouteUsage())) {constA=1, desc=Always pulsing when grown}
  n99 = Add(
   A=Multiply(A=Time(), B=ScalarParameter() {defaultValue=0.35, group=HealingCactus, parameterName=HC_Fruits_PulseSpeed})
   B=NamedRerouteUsage()
  )
  n100 = Multiply(A=ConstantBiasScale(None=Sine(None=@n99)), B=NamedRerouteUsage())
  NamedRerouteDeclaration(None=Multiply(A=Multiply(A=@n94, B=@n98), B=@n100)) {name=FruitPulseFresnel, nodeColor=<rgba 0 0.3672 1 1>}
  n101 = Divide(A=ComponentMask(None=NamedRerouteUsage()) {b=true}) {constB=64}
  NamedRerouteDeclaration(None=Saturate(None=Add(A=@n101) {constB=-0.1})) {name=RootMask, nodeColor=<rgba 1 0.4688 0 1>}
  NamedRerouteDeclaration(None=Multiply(A=@n100, B=@n98)) {name=FruitPulse, nodeColor=<rgba 1 0.2344 0 1>}
  n102 = TextureCoordinate() {coordinateIndex=3}
  n103 = Divide(A=AppendVector(A=TextureCoordinate() {coordinateIndex=2}, B=ComponentMask(None=@n102) {r=true})) {constB=100, desc=Scale down by 100 from Houdini}
  n104 = Constant3Vector() {constant=<rgba 0 -1 0 0>}
  NamedRerouteDeclaration(None=TransformPosition(None=Add(A=@n103, B=@n104)) {transformSourceType=TRANSFORMPOSSOURCE_Instance, transformType=TRANSFORMPOSSOURCE_World}) {name=Pivots, nodeColor=<rgba 0 1 0.5781 1>}
  n105 = WorldPosition()
  n106 = NamedRerouteUsage()
  n107 = Divide(A=NamedRerouteDeclaration(None=Distance(A=@n105.XYZ, B=@n106)) {name=PivotDistance, nodeColor=<rgba 1 0.4219 0 1>}) {constB=24, desc=24}
  n108 = Multiply(A=VertexNormalWS()) {constB=1.15}
  n109 = NamedRerouteUsage()
  NamedRerouteDeclaration(None=Multiply(A=@n107, B=LinearInterpolate(B=@n108, Alpha=@n109))) {name=WPO_Pulse, nodeColor=<rgba 1 0.4219 0 1>}
  n110 = ScalarParameter() {defaultValue=0.85, group=HealingCactus, parameterName=HC_SSSOpacity}
  NamedRerouteDeclaration(None=LinearInterpolate(
    A=Saturate(None=@n110)
    B=ScalarParameter() {defaultValue=0.3, group=HealingCactus, parameterName=HC_SSSOpacity_Fruits}
    Alpha=NamedRerouteUsage()
   )) {name=SSSOpacity_OUT, nodeColor=<rgba 1 0 1 1>}
  n111 = PreSkinnedNormal()
  NamedRerouteDeclaration(None=ShaderStageSwitch(PixelShader=VertexInterpolator(VS=@n111).PS, VertexShader=@n111)) {name=PS-LocalNrm, nodeColor=<rgba 0.3828 0 1 1>}
  n112 = CollectionParameter() {
   collection=&FortniteMaterialParameters
   parameterName=SunlightDirectionalVector
  }
  n113 = ComponentMask(None=@n112) {b=true, g=true, r=true}
  n114 = StaticSwitch(False=PixelNormalWS(), Value=StaticBoolParameter() {group=Slurp, parameterName=UseSlurpNormalDistortion})
  n115 = Multiply(A=Multiply(A=DotProduct(A=@n113, B=@n114)) {constB=2}, B=@n114)
  n116 = DotProduct(
   A=Normalize(VectorInput=Subtract(A=@n115, B=@n113))
   B=Multiply(A=MaterialFunctionCall() {materialFunction=&CameraDirectionVector}.Result) {constB=-1}
  )
  n117 = Power(Base=Max(A=@n116) {constB=0}, Exp=ScalarParameter() {defaultValue=1, group=Slurp, parameterName=Slurp_InnerScatteringPower})
  n118 = ScalarParameter() {defaultValue=0.4, group=Slurp, parameterName=Slurp_InnerScatteringScale}
  NamedRerouteDeclaration(None=Saturate(None=Multiply(A=@n117, B=@n118) {constB=0.4})) {name=InnerScattering Mask, nodeColor=<rgba 0 0.1797 1 1>}
  n119 = Multiply(A=@n116) {constB=-1}
  n120 = ScalarParameter() {group=Slurp, parameterName=Slurp_InnerBackScatteringBias}
  n121 = Power(
   Base=Max(A=Add(A=@n119, B=@n120)) {constB=0}
   Exp=ScalarParameter() {defaultValue=1, group=Slurp, parameterName=Slurp_InnerBackScatteringPower}
  )
  n122 = ScalarParameter() {defaultValue=0.4, group=Slurp, parameterName=Slurp_InnerBackScatteringScale}
  NamedRerouteDeclaration(None=Saturate(None=Multiply(A=@n121, B=@n122) {constB=0.55})) {name=BackScatteringMask, nodeColor=<rgba 0 0.7656 1 1>}
  n123 = ScalarParameter() {defaultValue=0.8, group=Slurp, parameterName=Slurp_TimeofDayScale_Morning}
  n124 = ScalarParameter() {defaultValue=1, group=Slurp, parameterName=Slurp_TimeofDayScale_Day}
  n125 = ScalarParameter() {defaultValue=0.7, group=Slurp, parameterName=Slurp_TimeofDayScale_Evening}
  n126 = ScalarParameter() {defaultValue=0.6, group=Slurp, parameterName=Slurp_TimeofDayScale_Night}
  NamedRerouteDeclaration(None=Multiply(
    A=@n22
    B=MaterialFunctionCall(Value_Morning=@n123, Value_Day=@n124, Value_Evening=@n125, Value_Night=@n126) {materialFunction=&Scalar_Time_of_Day}.Result
   )) {name=SlurpColor, nodeColor=<rgba 1 0.0938 0 1>}
  n127 = NamedRerouteUsage()
  n128 = Multiply(A=@n127, B=ScalarParameter() {defaultValue=0.4, group=HealingCactus, parameterName=HC_FruitsRoughnessMulti})
  NamedRerouteDeclaration(None=LinearInterpolate(A=@n127, B=Saturate(None=@n128), Alpha=NamedRerouteUsage())) {name=Roughness_OUT, nodeColor=<rgba 1 0.1172 0 1>}
  n129 = NamedRerouteUsage()
  n130 = Multiply(A=@n129, B=ScalarParameter() {defaultValue=3.5, group=HealingCactus, parameterName=HC_FruitsSpecularMulti})
  NamedRerouteDeclaration(None=LinearInterpolate(A=@n129, B=Saturate(None=@n130), Alpha=NamedRerouteUsage())) {name=Specular_OUT, nodeColor=<rgba 1 0.1172 0 1>}
  NamedRerouteDeclaration(None=@n6.Specular) {name=Specular_IN, nodeColor=<rgba 0 1 0.5781 1>}
  n131 = ScalarParameter() {group=HealingCactus_DEBUG, parameterName=DebugTime}
  n132 = Frac(None=Multiply(A=Time(), B=ScalarParameter() {defaultValue=1, group=HealingCactus_DEBUG, parameterName=DEBUG_Speed}))
  n133 = Subtract(
   A=Time()
   B=NamedRerouteDeclaration(None=ScalarParameter() {bUseCustomPrimitiveData=true, group=HealingCactus, parameterName=HC_GameTimeFromBP}) {name=GameTimeFromBP, nodeColor=<rgba 0.5469 0 1 1>}
  )
  n134 = ScalarParameter() {group=HealingCactus, parameterName=DEBUG_FruitState, primitiveDataIndex=1}
  n135 = ScalarParameter() {bUseCustomPrimitiveData=true, group=HealingCactus, parameterName=HC_FruitState, primitiveDataIndex=1}
  n136 = NamedRerouteUsage()
  n137 = Abs(None=NamedRerouteDeclaration(None=StaticSwitch(True=@n134, False=@n135, Value=@n136)) {name=FruitState, nodeColor=<rgba 0.0078 1 0 1>})
  NamedRerouteDeclaration(None=StaticSwitch(
    True=StaticSwitchParameter(True=@n131, False=@n132) {group=HealingCactus, parameterName=UseDebugTime}
    False=LinearInterpolate(B=@n133, Alpha=@n137)
    Value=NamedRerouteUsage()
   )) {name=TimePassed, nodeColor=<rgba 1 0 0.1875 1>}
  NamedRerouteDeclaration(None=LinearInterpolate(B=@n34, Alpha=@n35) {constA=1}) {name=FruitsColorChangeProgress, nodeColor=<rgba 0 1 0.4609 1>}
  n138 = NamedRerouteUsage()
  n139 = ScalarParameter() {defaultValue=0.3, desc=0.4, group=HealingCactus, parameterName=HC_Fruits_BurstDuration}
  n140 = NamedRerouteUsage()
  n141 = Subtract(
   A=StaticSwitch(True=@n138, False=Divide(A=@n138, B=@n139), Value=NamedRerouteUsage())
   B=LinearInterpolate(
    A=ScalarParameter() {defaultValue=1, desc=0.4, group=HealingCactus, parameterName=HC_Cactus_BurstHideTime}
    B=Frac(None=@n140)
    Alpha=NamedRerouteUsage()
   ) {constA=0.8, constB=0.9, desc=Cactus + fruit mask}
  ) {desc=Pops in random order}
  n142 = Constant() {r=1}
  n143 = If(A=@n141, "A > B"=Constant(), "A == B"=@n142, "A < B"=@n142) {constB=1, desc=Invisible after expanding}
  n144 = ScalarParameter() {bUseCustomPrimitiveData=true, desc=0.4, group=HealingCactus, parameterName=HC_VehicleCaused, primitiveDataIndex=2}
  n145 = Add(A=NamedRerouteUsage())
  NamedRerouteDeclaration(None=LinearInterpolate(A=LinearInterpolate(A=@n143, Alpha=@n144) {constA=1, constB=0}, Alpha=Saturate(None=@n145))) {name=OpacityMask_OUT, nodeColor=<rgba 0.0547 1 0 1>}
  n146 = CurveAtlasRowParameter(CurveTime=Saturate(None=@n141)) {
   atlas=&CA_Wasteland_HealingCactus
   curve=&C_Wasteland_HealingCactus_Fruits
   defaultValue=1
   group=HealingCactus
   parameterName=HC_Fruits_AnimCurve
  }
  n147 = LinearInterpolate(
   B=Multiply(A=VertexNormalWS()) {constB=10}
   Alpha=NamedRerouteDeclaration(None=@n146.R) {name=FruitPoppingAnimCurve, nodeColor=<rgba 1 0.2109 0 1>}
  ) {constB=8}
  NamedRerouteDeclaration(None=Multiply(A=Multiply(A=@n107, B=@n147), B=NamedRerouteUsage())) {name=WPO_Burst_Fruit, nodeColor=<rgba 0 0.2949 1 1>}
  n148 = NamedRerouteUsage()
  n149 = NamedRerouteUsage()
  n150 = NamedRerouteUsage()
  n151 = Constant()
  n152 = NamedRerouteDeclaration(None=Add(A=NamedRerouteUsage(), B=NamedRerouteUsage())) {name=WPO_Regrowth, nodeColor=<rgba 0 1 0.0241 1>}
  n153 = Constant()
  n154 = Add(A=NamedRerouteUsage(), B=NamedRerouteUsage())
  NamedRerouteDeclaration(None=Add(A=Add(A=@n148, B=@n149), B=If(A=@n150, B=@n151, "A > B"=@n152, "A == B"=@n153, "A < B"=@n154))) {name=WPO_OUT, nodeColor=<rgba 1 0 0.0703 1>}
  NamedRerouteUsage()
  NamedRerouteUsage()
  NamedRerouteUsage()
  NamedRerouteUsage()
  NamedRerouteUsage()
  NamedRerouteUsage()
  ScalarParameter() {defaultValue=2.5, group=HealingCactus, parameterName=m}
  ScalarParameter() {defaultValue=0.6, group=HealingCactus, parameterName=Spread}
  ScalarParameter() {defaultValue=3, group=HealingCactus, parameterName=Power}
  ScalarParameter() {defaultValue=0.2, group=HealingCactus, parameterName=HC_FruitsRoughness}
  ScalarParameter() {defaultValue=0.5, group=HealingCactus, parameterName=HC_FruitsSpecular}
  Subtract(
   A=WorldPosition().XYZ
   B=MaterialFunctionCall() {materialFunction=&ObjectPivotPoint}.Object Pivot Location
  )
 ]}
